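# Physics asset for the DarkStance cosmetic backpack, giving its key ring and three dangling keys collision shapes and joints for physical motion.
# Single capsules cover the ring base, ring and each key, while three capsules approximate the pack body, which has no constraint.
# The ring is jointed to its base and each key to the ring, with tight twist and per-joint asymmetric swing limits.
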PhysicsAsset /BRCosmetics/Accessories/FORT_Backpacks/Backpack_M_MED_DarkStance/Meshes/M_MED_DarkStance_Pack_Physics
bodies {
 base_KeyRing [{shapeType=Capsule, length=1.5, radius=1, shapeName=base_KeyRing_capsule, center=<0.4274 0 0>, rotation=<rot 0 -90 0>}]
 dyn_KeyRing [{
   shapeType=Capsule
   length=7.5
   radius=0.65
   shapeName=dyn_KeyRing_capsule
   center=<3.782 0.0773 -0.1911>
   rotation=<rot 0.1178 -89.9999 11.4688>
  }]
 dyn_LeftKey [{
   shapeType=Capsule
   length=5.3659
   radius=0.75
   shapeName=dyn_LeftKey_capsule
   center=<4.2104 -0.1029 0.0171>
   rotation=<rot -0.0295 87.8072 -89.6505>
  }]
 dyn_MidKey [{
   shapeType=Capsule
   length=7.4588
   radius=0.75
   shapeName=dyn_MidKey_capsule
   center=<5.2088 -0.2652 0.0164>
   rotation=<rot -0.0109 89.407 -89.8029>
  }]
 dyn_RightKey [{
   shapeType=Capsule
   length=5.2991
   radius=0.75
   shapeName=dyn_RightKey_capsule
   center=<4.2181 -0.1553 -0.01>
   rotation=<rot 0 -92.0707 -89.8435>
  }]
 pack [
  {
   shapeType=Capsule
   length=16.0545
   radius=3.7411
   shapeName=pack_capsule
   center=<0.497 -7.6487 -16.344>
   rotation=<rot 4.9797 -127.1314 84.5754>
  }
  {
   shapeType=Capsule
   length=16.0545
   radius=3.7411
   shapeName=pack_capsule
   center=<1.6145 -7.9773 -7.9169>
   rotation=<rot 5.092 -125.9302 129.6966>
  }
  {shapeType=Capsule, length=7, radius=6, shapeName=pack_capsule, center=<-0.9593 -3.1084 -11.3641>, rotation=<rot 88.6175 -0.0084 37.2462>}
 ]
}
constraints [
 {bone1Name=dyn_LeftKey, bone2Name=dyn_KeyRing, swing1LimitDegrees=60, swing2LimitDegrees=45, twistLimitDegrees=5}
 {bone1Name=dyn_MidKey, bone2Name=dyn_KeyRing, swing1LimitDegrees=140, swing2LimitDegrees=45, twistLimitDegrees=5}
 {bone1Name=dyn_RightKey, bone2Name=dyn_KeyRing, swing1LimitDegrees=120, swing2LimitDegrees=60, twistLimitDegrees=5}
 {bone1Name=dyn_KeyRing, bone2Name=base_KeyRing, swing1LimitDegrees=80, swing2LimitDegrees=7.5, twistLimitDegrees=5}
]
